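; Player pawn Blueprint for Zeus cinematics that exposes custom events to trigger bolt, jump, phase 2 and DBNO moments.
; ZeusBolt and ZeusJump activate abilities by tag through the pawn's Ability System Component.
; ZeusPhase2 sends itself a transformation gameplay event, while ZeusDBNO applies heavy fixed damage to itself.
; BeginPlay, overlap and Tick hold only disabled parent calls, and the construction script just calls its parent.
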(event EventBeginPlay
  (|CallParentFunction) ; This node is disabled and will not be called.\nDrag off pins to build functionality.
)

(event Collision|EventActorBeginOverlap
  (params (OtherActor :type Actor%20Object%20Reference))
  (|CallParentFunction OtherActor) ; This node is disabled and will not be called.\nDrag off pins to build functionality.
)

(event EventTick
  (params (DeltaSeconds :type Float%20%28single-precision%29))
  (|CallParentFunction DeltaSeconds) ; This node is disabled and will not be called.\nDrag off pins to build functionality.
)

(event Custom|ZeusBolt
  (Abilities|TryActivateAbilitiesbyTag (Ability|GetAbilitySystemComponent self) (Variables|Default|GetTagContainer_BoltAbility) true))

(event Custom|ZeusDBNO
  (Game|Damage|ApplyDamage self 6e+03 0 self))

(event Custom|ZeusPhase2
  (Ability|SendGameplayEventtoActor self (Variables|Default|GetEventTagForTransformation)))

(event Custom|ZeusJump
  (Abilities|TryActivateAbilitiesbyTag (Ability|GetAbilitySystemComponent self) (Variables|Default|GetTagContainer_JumpAbility) true))

(fn ConstructionScript
  (|CallParentFunction))
